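# Sound Cue named as a spot ambience of Queen of Thorns whispers for the AI AxisThrow feature's voice-over audio.
# The cue has no sound nodes and a null root, so it has nothing to play as shown.
# It retains only cue-level settings: reduced volume, a maximum distance and a single pre-effect bus send entry with no bus specified.
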SoundCue /FNM_AI_AxisThrow/Audio/VO/Ambience/IP_Spot_TangyRadish_QueenOfThorns_Whispers_Cue
root=null
nodes []
properties {maxDistance=8000, volumeMultiplier=0.4, preEffectBusSends [{}]}
