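# Looping stereo MetaSound that plays intermittent Queen of Thorns whisper VO as attenuated outdoor ambience.
# On play it concatenates two wave arrays and, after a short initial delay, picks a weighted random clip and plays it with slight random pitch shift.
# The player's nearly-finished trigger feeds back through a deferred trigger variable to start a random delay before the next pick, so whispers repeat indefinitely at irregular intervals.
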MetaSoundSource /FNM_AI_AxisThrow/Audio/VO/Ambience/IP_VO_TangyRadish_QueenOfThorns_Whispers_Cue
inputs [UE.Source.OnPlay {type=Trigger}]
interfaces [UE.OutputFormat.Stereo:1.0, UE.Source:1.1]
outputs [
 "UE.OutputFormat.Stereo.Audio:0" {type=Audio}
 "UE.OutputFormat.Stereo.Audio:1" {type=Audio}
]
pages [Default {
  input_defaults {UE.Source.OnPlay=null}
  nodes [
   n1 = VariableDeferredAccessor.Trigger(Variable=InitVariable.Trigger().Variable)
   n2 = InitVariable.WaveAsset:Array()
   n3 = InitVariable.WaveAsset:Array()
   n4 = UE.RandomTime(Next=@n1.Value) {also {defaults {Max=9, Min=3}}}
   TriggerDelay_0 = UE.Trigger Delay(In=@n4.On Next, Reset=@n4.On Reset, Delay Time=@n4.Value)
   TriggerDelay_1 = UE.Trigger Delay(In=^UE.Source.OnPlay) {also {defaults {Delay Time=3}}}
   TriggerAnyTriggerAny2_0 = "TriggerAny.Trigger Any (2)"(In 1=@TriggerDelay_0.Out, In 0=@TriggerDelay_1.Out)
   ArrayRandomGet_0 = Array.Random Get:WaveAsset:Array(
    In Array=Array.Concat:WaveAsset:Array(
     Trigger=^UE.Source.OnPlay
     Right Array=VariableAccessor.WaveAsset:Array(Variable=@n2.Variable).Value
     Left Array=VariableAccessor.WaveAsset:Array(Variable=@n3.Variable).Value
    ).Array
    Next=@TriggerAnyTriggerAny2_0.Out
    Weights=VariableAccessor.Float:Array(Variable=InitVariable.Float:Array().Variable).Value
   )
   n5 = UE.RandomFloat(Next=@ArrayRandomGet_0.On Next) {also {defaults {Max=0.5, Min=-0.5}}}
   WavePlayer_0 = UE.Wave Player:Stereo(Wave Asset=@ArrayRandomGet_0.Value, Play=@n5.On Next, Pitch Shift=@n5.Value) {also {version [1,1]}}
   VariableMutator.Trigger(Variable=@n1.Variable, Value=@WavePlayer_0.On Nearly Finished)
  ]
  links [{dst="@UE.OutputFormat.Stereo.Audio:0", src=WavePlayer_0.Out Left}, {dst="@UE.OutputFormat.Stereo.Audio:1", src=WavePlayer_0.Out Right}]
 }]
properties {
 assetImportData=&self:AssetImportData
 attenuationSettings=&/ATT_FMM_AxisThrow_QueenOfThrones_Whispers
 bDebug=true
 bLooping=true
 duration=10000
 maxDistance=18000
 numChannels=2
 outputFormat=Stereo
 sampleRate=48000
 soundClassObject=&Fort_Ambient_Outdoors_Athena
 virtualizationMode=PlayWhenSilent
 volume=0.6
}
